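# Editor utility that auto-places Mimosa footstep and eating SoundLibrary notifies on selected animation sequences, rejecting montages.
# Both entry functions read rows from data tables, optionally clear existing notifies, then save modified assets.
# SetFootNotifies samples bone poses in two passes, finding heel-height and toe-reach extremes, then marking steps near them.
# SetMouthNotifies sums Jaw, Lip and Smile curves over time, tagging opening starts whose peaks exceed a tolerance.
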EditorUtilityBlueprint /CosmeticCompanions/Assets/Audio_Shared/Blueprints/AddMimosaNotifies
graphs [
 AddMimosaJawNotifes {nodes [
   n1 = AddMimosaJawNotifes!() -> (
    then
    ClearExistingNotifies:bool=~
    SoundLibraryTag:Gameplay Tag Structure=(TagName="SoundLibrary.Mimosa.Foley.Eat")
    JawOpenTolerance:double=0.1
   ) {is=title K2Node_FunctionEntry}
   > n2 = Set MimosaDataTable(MimosaDataTable:Data Table Object Reference=&/Mimosa_AutoTagDataTable_Eat) -> (Output_Get:Data Table Object Reference)
   > Clear(TargetArray:Array of Object References=Get ModifiedAssets() -> (ModifiedAssets:Array of Object References))
   > n3 = GetSelectedAssets()
   > n4 = For Each Loop(Array:Array of Object References=@n3) -> (Array Element:Object Reference)
   n5 = Get ModifiedAssets()
   SaveLoadedAssets(execute=@n4.Completed, AssetsToSave=@n5, bOnlyIfIsDirty=true)
   > Clear(TargetArray=@n5)
   > Branch(Condition=Is Not Empty!(TargetArray=Get AnimMontages() -> (AnimMontages:Array of Strings)) -> (ReturnValue:bool))
   n6 = JoinStringArray(SourceArray=Get AnimMontages(), Separator="\r\n")
   > Show Message Dialog(
    Title=MakeLiteralText(Value="LOCTEXT(\"Montage Found\")")
    Message="To Text (String)"(InString=Append(A="This tool doesn't work with montages, the following items have not been edited:\r\n", B=@n6))
    MessageType=Ok
    DefaultValue=No
    MessageCategory=Warning
   )
   > Clear(TargetArray:Array of Strings=Get AnimMontages())
   n7 = Cast To AnimSequenceBase!(execute=@n4.LoopBody, Object=@n4) -> (then, CastFailed, AsAnim Sequence Base:Anim Sequence Base Object Reference) {is=title K2Node_DynamicCast}
   > n8 = Cast To AnimMontage!(execute, Object=@n7) -> (then, CastFailed, AsAnim Montage:Anim Montage Object Reference) {is=title K2Node_DynamicCast}
   TransactObject(execute=@n8.CastFailed, Object=@n4)
   > n9 = GetDataTableRowNames(Table=@n2)
   > n10 = For Each Loop(Array:Array of Names=@n9) -> (Array Element:Name)
   > n11 = Get Data Table Row!(execute:Exec, DataTable:Data Table Object Reference=@n2, RowName:Name=@n10) -> (then:Exec, RowNotFound:Exec, ReturnValue:Mimosa Autotag Audio Notify Struct Structure)
   > n12 = Branch(Condition=@n1)
   n13 = Break Mimosa Autotag Audio Notify Struct!(MimosaAutotagAudioNotify_Struct:"Mimosa Autotag Audio Notify Struct Structure (by ref)"=@n11) -> (
    HeelBone_15_#1:Name
    ToeBone_16_#2:Name
    NotifyClass_21_#3:Anim Notify Class Reference
    GameplayTag_22_#4:Gameplay Tag Structure
    AnimNotifyTrackName_18_#5:Name
    FootBone_25_#6:int
    Bypass_27_#7:bool=~
   ) {is=title K2Node_BreakStruct}
   > n14 = Remove Existing Notifies!(
    execute:Exec
    AnimationSequenceBase:Anim Sequence Base Object Reference=@n7
    ParentClass:Object Class Reference=@n13.NotifyClass_21_#3
    AudioFootstepNotify_Struct:"Mimosa Auto Tag Settings Structure (by ref)"
   ) -> (then:Exec) {is=title K2Node_MacroInstance}
   SetMouthNotifies!(
    execute=[@n14.then, @n12.else]
    self
    Anim:Anim Sequence Base Object Reference=@n7
    Anim Notify Class:Anim Notify Class Reference=@n13.NotifyClass_21_#3
    SoundLibraryTag:Gameplay Tag Structure=Get SoundLibraryTag() -> (SoundLibraryTag:Gameplay Tag Structure)
    Track:Name=@n13.AnimNotifyTrackName_18_#5
    ByPass:bool=@n13.Bypass_27_#7
    JawOpenTolerence:double=Get JawOpenTolerance() -> (JawOpenTolerance:double)
   ) -> (then)
   Append(A="- ", B=GetDisplayName(Object=@n8))
  ]}
 AddMimosaNotifies {nodes [
   n1 = AddMimosaNotifies!() -> (then, bClearExistingNotifies:bool=~, CharacterType:E_MimosaAutoTag_AnimType Enum=NewEnumerator0) {is=title K2Node_FunctionEntry}
   n2 = Select!(
    NewEnumerator0:Data Table Object Reference=&/Mimosa_AutoTagDataTable_Quadruped
    NewEnumerator1:Data Table Object Reference=&/Mimosa_AutoTagDataTable_Biped
    Index:E_MimosaAutoTag_AnimType Enum=@n1.CharacterType
   ) -> (ReturnValue:Data Table Object Reference)
   > n3 = Is Valid(InputObject=@n2)
   Show Message Dialog(
    execute=@n3.Is Not Valid
    Title=MakeLiteralText(Value="LOCTEXT(\"Pardon the dust!\")")
    Message=MakeLiteralText(Value="LOCTEXT(\"The bark function is under construction, and doesn\\'t currently work. Sorry for the block!\")")
    MessageType=Ok
    DefaultValue=No
    MessageCategory=Warning
   )
   n4 = Set MimosaDataTable(execute=@n3.Is Valid, MimosaDataTable=@n2)
   > Clear(TargetArray:Array of Object References=Get ModifiedAssets())
   > n5 = GetSelectedAssets()
   > n6 = For Each Loop(Array:Array of Object References=@n5) -> (Array Element:Object Reference)
   n7 = Get ModifiedAssets()
   SaveLoadedAssets(execute=@n6.Completed, AssetsToSave=@n7, bOnlyIfIsDirty=true)
   > Clear(TargetArray=@n7)
   > Branch(Condition=Is Not Empty(TargetArray=Get AnimMontages()))
   n8 = JoinStringArray(SourceArray=Get AnimMontages(), Separator="\r\n")
   > Show Message Dialog(
    Title=MakeLiteralText(Value="LOCTEXT(\"Montage Found\")")
    Message="To Text (String)"(InString=Append(A="This tool doesn't work with montages, the following items have not been edited:\r\n", B=@n8))
    MessageType=Ok
    DefaultValue=No
    MessageCategory=Warning
   )
   > Clear(TargetArray:Array of Strings=Get AnimMontages())
   n9 = Cast To AnimSequenceBase(execute=@n6.LoopBody, Object=@n6)
   > n10 = Cast To AnimMontage(Object=@n9)
   > Add(TargetArray=Get AnimMontages(), NewItem=Append(A="- ", B=GetDisplayName(Object=@n10))) {is=title K2Node_CallArrayFunction}
   TransactObject(execute=@n10.CastFailed, Object=@n6)
   > n11 = GetDataTableRowNames(Table=@n4)
   > n12 = For Each Loop(Array:Array of Names=@n11) -> (Array Element:Name)
   > n13 = Get Data Table Row(DataTable=@n4, RowName=@n12)
   > n14 = Branch(Condition=@n1)
   n15 = Break Mimosa Autotag Audio Notify Struct(MimosaAutotagAudioNotify_Struct=@n13) -> (Bypass_27_#7=~)
   > n16 = Remove Existing Notifies(AnimationSequenceBase=@n9, ParentClass=@n15.NotifyClass_21_#3)
   SetFootNotifies!(
    execute=[@n16.then, @n14.else]
    self
    Anim:Anim Sequence Base Object Reference=@n9
    HeelBone:Name=@n15
    ToeBone:Name=@n15.ToeBone_16_#2
    NotifyClass:Anim Notify Class Reference=@n15.NotifyClass_21_#3
    GameplayCue:Gameplay Tag Structure=@n15.GameplayTag_22_#4
    TrackName:Name=@n15.AnimNotifyTrackName_18_#5
    FrameOffset:int=0
    FootBone:int=@n15.FootBone_25_#6
    Bypass:bool=@n15.Bypass_27_#7
   ) -> (then)
  ]}
 Debug string {nodes [
   n1 = Debug string!() -> (then, CurrentToeVal:double=~, MaxToeVal:double=~, CurrentHeelVal:double=~, MaxHeelVal:double=~) {is=title K2Node_FunctionEntry}
   > PrintString(InString=Append!(
     A="***Current Toe: "
     B="To String (Float)"(InDouble=@n1)
     C:String=" - Max Toe: "
     D:String="To String (Float)"(InDouble=@n1.MaxToeVal)
     E:String=" - Current Heel: "
     F:String="To String (Float)"(InDouble=@n1.CurrentHeelVal)
     G:String=" - Max Heel: "
     H:String="To String (Float)"(InDouble=@n1.MaxHeelVal)
     I:String
     J:String
    ) -> (ReturnValue))
  ]}
 EventGraph {links [], nodes []}
 IterateAnimation {nodes [IterateAnimation() {is=title K2Node_FunctionEntry}]}
 RemoveExistingNotifies {nodes [
   n1 = Inputs!() -> (
    execute
    AnimationSequenceBase:Anim Sequence Base Object Reference
    ParentClass:Object Class Reference
    AudioFootstepNotify_Struct:"Mimosa Auto Tag Settings Structure (by ref)"
   )
   > n2 = For Each Loop(Array=GetAnimationTrackNames!(AnimationSequenceBase:Anim Sequence Base Object Reference=@n1) -> (TrackNames:Array of Names))
   Outputs!(then:Exec=@n2.Completed)
   > Set bTrackContainsOutsideNotify(bTrackContainsOutsideNotify:bool=false) -> (Output_Get:bool)
   n3 = GetAnimationNotifyEventsForTrack!(AnimationSequenceBase:Anim Sequence Base Object Reference=@n1, NotifyTrackName:Name=@n2) -> (Events:Array of Anim Notify Event Structures)
   > n4 = Branch(Condition=Is Empty!(TargetArray:Array of Anim Notify Event Structures=@n3) -> (ReturnValue:bool))
   n5 = For Each Loop(Exec=@n4.else, Array:Array of Anim Notify Event Structures=@n3) -> (Array Element:Anim Notify Event Structure)
   n6 = Branch(execute=@n5.Completed, Condition=Get bTrackContainsOutsideNotify() -> (bTrackContainsOutsideNotify:bool))
   > PrintString(InString=Append!(
     A="Track: "
     B="To String (Name)"(InName=@n2)
     C=" contains notifes other than notify class "
     D=Get Class Display Name(Class=@n1.ParentClass)
     E=. Cannot remove previous notifies from track. Manually remove them instead.
    ) -> (ReturnValue))
   RemoveAnimationNotifyEventsByTrack!(execute=@n6.else, AnimationSequenceBase:Anim Sequence Base Object Reference=@n1, NotifyTrackName:Name=@n2) -> (then, ReturnValue:int=0)
   n7 = Break Anim Notify Event!(AnimNotifyEvent:"Anim Notify Event Structure (by ref)"=@n5) -> (
    TriggerWeightThreshold:float=~
    NotifyName:Name
    Notify:Anim Notify Object Reference
    NotifyStateClass:Anim Notify State Object Reference
    MontageTickType:EMontageNotifyTickType Enum
    NotifyTriggerChance:float=~
    NotifyFilterType:ENotifyFilterType Enum
    NotifyFilterLOD:int
    bCanBeFilteredViaRequest:bool=~
    bTriggerOnDedicatedServer:bool=~
    bTriggerOnFollower:bool=~
   ) {is=title K2Node_BreakStruct}
   n8 = Branch(
    execute=@n5.LoopBody
    Condition=ClassIsChildOf(TestClass=Get Class(Object=@n7.Notify) -> (ReturnValue:Anim Notify Class Reference), ParentClass=@n1.ParentClass)
   )
   Set bTrackContainsOutsideNotify(execute=@n8.else, bTrackContainsOutsideNotify=true)
  ]}
 SetFootNotifies {nodes [
   n1 = SetFootNotifies!() -> (
    then
    Anim:Anim Sequence Base Object Reference
    HeelBone:Name
    ToeBone:Name
    NotifyClass:Anim Notify Class Reference
    GameplayCue:Gameplay Tag Structure
    TrackName:Name
    FrameOffset:int
    FootBone:int
    Bypass:bool=~
   ) {is=title K2Node_FunctionEntry}
   > n2 = Branch(Condition=@n1.Bypass)
   Add(execute=@n2.else, TargetArray:Array of Object References=Get ModifiedAssets(), NewItem:"Object Reference (by ref)"=@n1)
   n3 = Get Anim() -> (Anim:Anim Sequence Base Object Reference)
   > n4 = Branch(Condition=Contains Item(
     TargetArray=GetAnimationNotifyTrackNames!(AnimationSequenceBase:Anim Sequence Base Object Reference=@n3) -> (TrackNames:Array of Names)
     ItemToFind=Get TrackName() -> (TrackName:Name=None)
    ))
   n5 = Sequence(execute=[@n34.then, @n4.then])
   n6 = Get SequenceLength!(self:Anim Sequence Base Object Reference=Get Anim()) -> (SequenceLength:float)
   n7 = Get IncrementVal() -> (IncrementVal:double)
   > n8 = For Loop(LastIndex=Truncate(A=float / float(A=@n6, B=@n7)))
   Set CurrentAnimTime(execute=@n8.Completed, CurrentAnimTime:double=0) -> (Output_Get:double)
   n9 = GetAnimPoseAtTime!(
    execute=@n8.LoopBody
    AnimationSequenceBase:Anim Sequence Base Object Reference=Get Anim()
    Time:double=Get CurrentAnimTime() -> (CurrentAnimTime:double)
    EvaluationOptions:Anim Pose Evaluation Options Structure=Make Anim Pose Evaluation Options!(
     EvaluationType:EAnimDataEvalType Enum=Raw
     bShouldRetarget:bool=True
     bExtractRootMotion:bool=False
     bIncorporateRootMotionIntoPose:bool=True
     OptionalSkeletalMesh:Skeletal Mesh Object Reference
     bRetrieveAdditiveAsFullPose:bool=True
     bEvaluateCurves:bool=True
    ) -> (AnimPoseEvaluationOptions:Anim Pose Evaluation Options Structure) {is=title K2Node_MakeStruct}
   ) -> (then, Pose:Anim Pose Structure)
   > n10 = Sequence!(execute) -> (then_0, then_1, then_2:Exec)
   Set CurrentAnimTime(execute=@n10.then_2, CurrentAnimTime=float + float(A=Get CurrentAnimTime(), B=Get IncrementVal()))
   n11 = Get HeelBone() -> (HeelBone:Name=None)
   n12 = BreakVector(InVec=BreakTransform(InTransform=GetBonePose!(Pose:"Anim Pose Structure (by ref)"=@n9, BoneName:Name=@n11, Space:EAnimPoseSpaces Enum=World) -> (ReturnValue:"Transform (by ref)")))
   Branch(execute=@n10.then_0, Condition="float < float"(A=@n12.Z, B=Get LowestBoneHeight() -> (LowestBoneHeight:double)))
   > Set LowestBoneHeight(LowestBoneHeight:double=@n12.Z) -> (Output_Get:double)
   n13 = Get ToeBone() -> (ToeBone:Name=None)
   n14 = BreakVector(InVec=BreakTransform(InTransform=GetBonePose(Pose=@n9, BoneName=@n13, Space=World)))
   Branch(execute=@n10.then_1, Condition="float > float"(A=@n14.Y, B=Get FarthestToePoint() -> (FarthestToePoint:double)))
   > Set FarthestToePoint(FarthestToePoint:double=@n14.Y) -> (Output_Get:double)
   n15 = Get IncrementVal()
   n16 = Get SequenceLength(self=Get Anim())
   For Loop(execute=@n5.then_1, LastIndex=Truncate(A=float / float(A=@n16, B=@n15)))
   > n17 = GetAnimPoseAtTime(
    AnimationSequenceBase=Get Anim()
    Time=Get CurrentAnimTime()
    EvaluationOptions=Make Anim Pose Evaluation Options(
     EvaluationType=Raw
     bShouldRetarget=True
     bExtractRootMotion=False
     bIncorporateRootMotionIntoPose=True
     bRetrieveAdditiveAsFullPose=True
     bEvaluateCurves=True
    )
   )
   n18 = Get HeelBone() -> (HeelBone=None)
   n19 = BreakVector(InVec=BreakTransform(InTransform=GetBonePose(Pose=@n17, BoneName=@n18, Space=World)))
   n20 = Get ToeBone() -> (ToeBone=None)
   n21 = BreakVector(InVec=BreakTransform(InTransform=GetBonePose(Pose=@n17, BoneName=@n20, Space=World)))
   > Debug string!(
    execute
    self
    CurrentToeVal:double=@n21.Y
    MaxToeVal:double=Get FarthestToePoint()
    CurrentHeelVal:double=@n19.Z
    MaxHeelVal:double=Get LowestBoneHeight()
   ) -> (then) {is=title K2Node_CallFunction}
   > n22 = Branch(Condition=Get IsStepping() -> (IsStepping:bool))
   n23 = float + float(A=Get LowestBoneHeight(), B=2)
   n24 = "float < float"(A=@n19.Z, B=@n23)
   > n25 = Branch(Condition=@n24)
   n26 = float - float(A=Get FarthestToePoint(), B=1)
   n27 = "float > float"(A=@n21.Y, B=@n26)
   > n28 = Branch(Condition=@n27)
   > Set IsStepping(IsStepping:bool=false) -> (Output_Get:bool)
   > n29 = AddAnimationNotifyEvent!(
    execute
    AnimationSequenceBase:Anim Sequence Base Object Reference=Get Anim()
    NotifyTrackName:Name=Get TrackName() -> (TrackName=None)
    StartTime:float=Get CurrentAnimTime()
    NotifyClass:Anim Notify Class Reference=Get NotifyClass() -> (NotifyClass:Anim Notify Class Reference)
   ) -> (then, ReturnValue:Anim Notify Object Reference)
   > n30 = Cast To AnimNotify_SoundLibrary!(execute, Object=@n29) -> (then, CastFailed, AsAnim Notify Sound Library:Anim Notify Sound Library Object Reference) {is=title K2Node_DynamicCast}
   > n31 = Create Sound Library Context!(
    execute:Exec
    WorldContextObject:Object Reference
    Class:Sound Library Context Class Reference=/Script/CosmeticCompanionsRuntime.CompanionSoundLibraryAnimContext
    AnimSettings:Animation Notify Settings Structure=Get Anim Settings() -> (Anim Settings:Animation Notify Settings Structure)
    Animation:Anim Sequence Base Weak object
    Actor:Actor Weak object
    BaseEventName:Gameplay Tag Structure
    bForwardToActorOwner:bool=True
    bForwardToAttachedActors:bool=True
    bAutoResetContext:bool=True
   ) -> (then:Exec, ReturnValue:Companion Anim Context Object Reference) {is=title K2Node_CreateSoundLibraryContext}
   > n32 = SetEditorProperty(Object=@n30, PropertyName=SoundContext, PropertyValue:Companion Anim Context Object Reference=@n31)
   Set CurrentAnimTime(execute=[@n32.then, @n30.CastFailed], CurrentAnimTime=float + float(A=Get IncrementVal(), B=Get CurrentAnimTime()))
   Set CurrentAnimTime(execute=[@n25.else, @n28.else, @n33.then], CurrentAnimTime=float + float(A=Get IncrementVal(), B=Get CurrentAnimTime()))
   n33 = Set IsStepping(execute=@n22.else, IsStepping=AND Boolean(A="float > float"(A=@n19.Z, B=@n23), B="float < float"(A=@n21.Y, B=@n26)))
   n34 = AddAnimationNotifyTrack!(
    execute=@n4.else
    AnimationSequenceBase:Anim Sequence Base Object Reference=Get Anim()
    NotifyTrackName:Name=Get TrackName() -> (TrackName=None)
    TrackColor:Linear Color Structure=(R=1,G=1,B=1,A=1)
   ) -> (then)
   GetNumFrames!(AnimationSequenceBase:Anim Sequence Base Object Reference=Get Anim()) -> (NumFrames:int=0)
   PrintString(InString=Hello)
   AND Boolean(A=@n24, B=@n27)
  ]}
 SetMouthNotifies {nodes [
   n1 = SetMouthNotifies!() -> (
    then
    Anim:Anim Sequence Base Object Reference
    Anim Notify Class:Anim Notify Class Reference
    SoundLibraryTag:Gameplay Tag Structure
    Track:Name
    ByPass:bool=~
    JawOpenTolerence:double=~
   ) {is=title K2Node_FunctionEntry}
   > n2 = Branch(Condition=@n1.ByPass)
   Add(execute=@n2.else, TargetArray=Get ModifiedAssets(), NewItem=@n1)
   n3 = Get Anim()
   > n4 = Branch(Condition=Contains Item(TargetArray=GetAnimationNotifyTrackNames(AnimationSequenceBase=@n3), ItemToFind=Get Track() -> (Track:Name=None)))
   n5 = For Each Loop(
    Exec=[@n23.then, @n4.then]
    Array:Array of Names=GetAnimationCurveNames!(AnimationSequenceBase:Anim Sequence Base Object Reference=Get Anim(), CurveType:ERawCurveTrackTypes Enum=RCT_Float) -> (CurveNames:Array of Names)
   ) -> (Array Element:Name)
   n6 = "To String (Name)"(InName=@n5)
   > Branch(Condition=OR Boolean!(A=Contains(SearchIn=@n6, Substring=Jaw), B=Contains(SearchIn=@n6, Substring=Lip), C:bool=Contains(SearchIn=@n6, Substring=Smile)) -> (ReturnValue))
   > Add(TargetArray:Array of Names=Get CurvesToWatch() -> (CurvesToWatch:Array of Names), NewItem:"Name (by ref)"=@n5)
   n7 = Get SequenceLength(self=Get Anim())
   n8 = Get IncrementVal()
   > n9 = For Loop(LastIndex=Truncate(A=float / float(A=@n7, B=@n8)))
   Set CurrentAnimTime(execute=@n9.Completed, CurrentAnimTime=0)
   > n10 = For Each Loop(Array:"Array of Float (double-precision)s"=Get TagAnimTimes() -> (TagAnimTimes:"Array of Float (double-precision)s")) -> (Array Element:double)
   Clear(
    execute=@n10.Completed
    TargetArray:"Array of Float (double-precision)s"=Get JawPeakHeights() -> (JawPeakHeights:"Array of Float (double-precision)s")
   )
   > Clear(TargetArray=Get TagAnimTimes())
   > Set PreviousJawPose(PreviousJawPose:double=0) -> (Output_Get:double)
   n11 = Get PrevTagAnimTime() -> (PrevTagAnimTime:double)
   Branch(execute=@n10.LoopBody, Condition="float > float"(A=@n10, B=float + float(A=@n11, B=0.01)))
   > Set PrevTagAnimTime(PrevTagAnimTime:double=@n10) -> (Output_Get:double)
   n12 = Get JawPeakHeights()
   > Branch(Condition="float > float"(
     A="Get (a copy)"(Array:"Array of Float (double-precision)s"=@n12, Dimension 1=@n10.Array Index) -> (Output:double)
     B=Get JawOpenTolerence() -> (JawOpenTolerence:double)
    ))
   > n13 = AddAnimationNotifyEvent(
    AnimationSequenceBase=Get Anim()
    NotifyTrackName=Get Track() -> (Track=None)
    StartTime=@n10
    NotifyClass=Get Anim Notify Class() -> (Anim Notify Class:Anim Notify Class Reference)
   )
   > n14 = Cast To AnimNotify_SoundLibrary(Object=@n13)
   n15 = Break Animation Notify Settings!(SoundLibraryAnimContextSettings:"Animation Notify Settings Structure (by ref)"=Get Anim Settings()) -> (
    AnimRateThreshold:Float Interval Structure
    EventName:Gameplay Tag Structure
    bOnThisActor:bool=~
    bAttachToActor:bool=~
    bStopWhenAttachedDestroyed:bool=~
    Offset:Vector=~
    AttachName:Name
    FadeTime:float=~
    NotifyTriggerChance:float=~
    VolumeMultiplier:float=~
   ) {is=title K2Node_BreakStruct}
   > n16 = Create Sound Library Context(
    Class=&CompanionSoundLibraryAnimContext
    AnimSettings=Make Animation Notify Settings!(
     AnimRateThreshold:Float Interval Structure=@n15
     EventName:Gameplay Tag Structure=Get SoundLibraryTag()
     bOnThisActor:bool=@n15.bOnThisActor
     bAttachToActor:bool=@n15.bAttachToActor
     bStopWhenAttachedDestroyed:bool=@n15.bStopWhenAttachedDestroyed
     Offset:Vector=@n15.Offset
     AttachName:Name=@n15.AttachName
     FadeTime:float=@n15.FadeTime
     NotifyTriggerChance:float=@n15.NotifyTriggerChance
     VolumeMultiplier:float=@n15.VolumeMultiplier
    ) -> (SoundLibraryAnimContextSettings:Animation Notify Settings Structure) {is=title K2Node_MakeStruct}
    bForwardToActorOwner=True
    bForwardToAttachedActors=True
    bAutoResetContext=True
   )
   > SetEditorProperty(Object=@n14, PropertyName=SoundContext, PropertyValue=@n16)
   n17 = For Each Loop(Exec=@n9.LoopBody, Array:Array of Names=Get CurvesToWatch()) -> (Array Element:Name)
   n18 = Sequence(execute=@n17.Completed)
   > n19 = Branch(Condition="float > float"(A=Get CurrentMouthPose() -> (CurrentMouthPose:double), B=Get PreviousJawPose() -> (PreviousJawPose:double)))
   > Do Once(Reset=@n19.else, Start Closed=true)
   > Add(TargetArray:"Array of Float (double-precision)s"=Get TagAnimTimes(), NewItem:"Float (double-precision) (by ref)"=Get CurrentAnimTime())
   n20 = Branch(execute=@n18.then_1, Condition="float < float"(A=Get CurrentMouthPose(), B=Get PreviousJawPose()))
   > Do Once(Reset=@n20.else, Start Closed=true)
   > n21 = Set JawPeak(JawPeak:double=Get CurrentMouthPose()) -> (Output_Get:double)
   > Add(TargetArray=Get JawPeakHeights(), NewItem=@n21)
   > PrintString(InString=Append!(A=Found Peak At Value, B="To String (Float)"(InDouble=@n21)) -> (ReturnValue))
   Set PreviousJawPose(execute=@n18.then_2, PreviousJawPose=Get CurrentMouthPose())
   > Set CurrentAnimTime(CurrentAnimTime=float + float(A=Get CurrentAnimTime(), B=Get IncrementVal()))
   > Set CurrentMouthPose(CurrentMouthPose:double=0) -> (Output_Get:double)
   n22 = GetFloatValueAtTime!(
    execute=@n17.LoopBody
    AnimationSequenceBase:Anim Sequence Base Object Reference=Get Anim()
    CurveName:Name=@n17
    Time:float=Get CurrentAnimTime()
   ) -> (then, ReturnValue:float)
   > Set CurrentMouthPose(CurrentMouthPose=float + float(A=@n22, B=Get CurrentMouthPose()))
   n23 = AddAnimationNotifyTrack(execute=@n4.else, AnimationSequenceBase=Get Anim(), NotifyTrackName=Get Track() -> (Track=None), TrackColor=(R=1,G=1,B=1,A=1))
  ]}
]
variables [
 Anim Settings {
  type=SoundLibraryAnimContextSettings
  default {
   attachName=None
   bAttachToActor=false
   bOnThisActor=false
   bStopWhenAttachedDestroyed=true
   fadeTime=0.5
   notifyTriggerChance=1
   previewAnimationRate=1
   volumeMultiplier=1
   animRateThreshold {max=0, min=0}
   eventName {tagName=SoundLibrary.Mimosa.Foley.Step}
   offset=<0 0 0>
  }
 }
 AnimMontages {type="array<string>", default []}
 bTrackContainsOutsideNotify {type=boolean}
 FootstepTag {type=SoundLibraryTag, default {tagName=SoundLibrary.Mimosa.Foley.Step}}
 IncrementVal {type=number, default=0.001}
 MimosaDataTable {type=DataTable, default=null}
 ModifiedAssets {type="array<Object>", default []}
]
