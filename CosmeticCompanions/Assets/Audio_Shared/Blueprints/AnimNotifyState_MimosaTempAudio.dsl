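; Anim notify state in CosmeticCompanions that plays a temporary sound, TempAudio, attached to the mesh for the length of the notify window.
; On begin it spawns the attached audio component and sets its component tag to the sound's display name.
; On end it finds the owner's audio components with that tag and briefly fades each out, so no component reference is stored on the notify.
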(fn Received_NotifyBegin
  (params (MeshComp :type Skeletal%20Mesh%20Component%20Object%20Reference) (Animation :type Anim%20Sequence%20Base%20Object%20Reference) (TotalDuration :type Float%20%28single-precision%29) (EventReference :type Anim%20Notify%20Event%20Reference%20Structure%20%28by%20ref%29))
  (bind _tempaudio (Variables|Default|GetTempAudio))
  (bind _returnvalue (Utilities|IsValid _tempaudio :!class /Script/Engine.KismetSystemLibrary))
  (if _returnvalue
    (bind _returnvalue_2 (Audio|SpawnSoundAttached _tempaudio MeshComp))
    (Utilities|IsValid _returnvalue_2
      (:Is%20Valid
        (Variables|/Script/Engine.ActorComponent|SetComponentTags (Utilities|Array|MakeArray (Utilities|String|StringToName (Utilities|GetDisplayName _tempaudio))) _returnvalue_2)))
    (else
      _))
  (return _returnvalue))

(fn Received_NotifyEnd
  (params (MeshComp :type Skeletal%20Mesh%20Component%20Object%20Reference) (Animation :type Anim%20Sequence%20Base%20Object%20Reference) (EventReference :type Anim%20Notify%20Event%20Reference%20Structure%20%28by%20ref%29))
  (bind _returnvalue (|CallParentFunction MeshComp Animation EventReference))
  (CallFunction|StopAudio (Components|GetOwner MeshComp))
  (return _returnvalue))

(fn StopAudio
  (params (Actor :type Actor%20Object%20Reference))
  (bind _tempaudio (Variables|Default|GetTempAudio))
  (if (Utilities|IsValid _tempaudio :!class /Script/Engine.KismetSystemLibrary)
    (for _array_element (Actor|GetComponentsbyTag Actor "/Script/Engine.AudioComponent" (Utilities|String|StringToName (Utilities|GetDisplayName _tempaudio)))
      (Audio|Components|Audio|FadeOut _array_element 0.25 0.0 "Linear" :!class /Script/Engine.AudioComponent))))
