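# Editor utility that finds and replaces attach names in Sound Library anim notifies on selected animation sequences and montages.
# FindGivenAttachName matches notify context AttachNames by substring, lists affected assets in a dialog, syncs the Content Browser to them and asks for confirmation.
# On Yes, ReplaceNotifyAttachNames rescans the selection, rewrites matching settings with transaction support, reports edits and saves dirty assets.
# Its bone-existence check is not wired into the execution flow.
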EditorUtilityBlueprint /CosmeticCompanions/Assets/Audio_Shared/Blueprints/Mimosa_Replace_AttachNames
graphs [
 CreateDialogueAndFindPaths {nodes [
   n1 = CreateDialogueAndFindPaths!() -> (then, DB Text:String, Array:Array of Object References, MessageType:EAppMsgType Enum=Ok, DB Prompt:String, Replace Attach Name:String) {is=title K2Node_FunctionEntry}
   > n2 = Branch(Condition=Is Not Empty!(TargetArray:Array of Object References=@n1.Array) -> (ReturnValue:bool))
   > n3 = For Each Loop(Array:Array of Object References=@n1.Array) -> (Array Element:Object Reference)
   n4 = JoinStringArray(SourceArray=Get SequenceNames() -> (SequenceNames:Array of Strings), Separator="\r\n")
   n5 = Get AttachNameToFind() -> (AttachNameToFind:String)
   n6 = Show Message Dialog(
    execute=@n3.Completed
    Title=MakeLiteralText(Value="LOCTEXT(\"Assets Found with Attach Points\")")
    Message="To Text (String)"(InString=Append!(A=@n1, B=@n5, C:String="\r\n", D:String=@n4, E:String=@n1.DB Prompt, F:String=@n1.Replace Attach Name) -> (ReturnValue))
    MessageType=@n1.MessageType
    DefaultValue=No
    MessageCategory=Warning
   )
   > Return Node!(execute, FoundObjectPaths:Array of Strings=Get ObjectPaths() -> (ObjectPaths:Array of Strings), PromptResponse:EAppReturnType Enum=@n6)
   Add(execute=@n3.LoopBody, TargetArray=Get SequenceNames(), NewItem=GetDisplayName(Object=@n3)) {is=title K2Node_CallArrayFunction}
   > Add(TargetArray=Get ObjectPaths(), NewItem=Get Object Path String(Object=@n3))
   n7 = Get AttachNameToFind()
   Show Message Dialog(
    execute=@n2.else
    Title=MakeLiteralText(Value="LOCTEXT(\"Assets Found with Attach Points\")")
    Message="To Text (String)"(InString=Append!(A="No Assets Found with Attach Name: ", B=@n7) -> (ReturnValue))
    MessageType=Ok
    DefaultValue=No
    MessageCategory=Warning
   )
  ]}
 EventGraph {links [], nodes []}
 FindAndReplaceSoundLibAttachBones {nodes [
   n1 = FindAndReplaceSoundLibAttachBones!() -> (then, Attach Name To Find:String, New Attach Name:String) {is=title K2Node_FunctionEntry}
   > FindGivenAttachName!(execute, self, Find Attach Name:String=@n1, New Attach Name:String=@n1.New Attach Name) -> (then)
  ]}
 FindAnimNotifiesWithJawOrPelvisUsage {nodes [
   n1 = FindAnimNotifiesWithJawOrPelvisUsage!() -> (then, NotifyContext:Animation Notify Context Object Reference) {is=title K2Node_FunctionEntry}
   > n2 = Set AttachName(AttachName:Name=Break Animation Notify Settings!(SoundLibraryAnimContextSettings:"Animation Notify Settings Structure (by ref)"=Get AnimSettings!(self:Animation Notify Context Object Reference=@n1) -> (AnimSettings:Animation Notify Settings Structure)) -> (
     AnimRateThreshold:Float Interval Structure
     EventName:Gameplay Tag Structure
     bOnThisActor:bool=~
     bAttachToActor:bool=~
     bStopWhenAttachedDestroyed:bool=~
     Offset:Vector=~
     AttachName:Name
     FadeTime:float=~
     NotifyTriggerChance:float=~
     VolumeMultiplier:float=~
    ) {is=title K2Node_BreakStruct}.AttachName) -> (Output_Get:Name=None)
   > Branch(Condition=Contains(SearchIn="To String (Name)"(InName=@n2), Substring=Get AttachNameToFind()))
   > Add(TargetArray:Array of Names=Get AttachNamesArray() -> (AttachNamesArray:Array of Names), NewItem:"Name (by ref)"=@n2)
   > Add(
    TargetArray:Array of Anim Notify Event Structures=Get NotifiesToUpdate() -> (NotifiesToUpdate:Array of Anim Notify Event Structures)
    NewItem:"Anim Notify Event Structure (by ref)"=Get Current Notify() -> (Current Notify:Anim Notify Event Structure)
   )
  ]}
 FindGivenAttachName {nodes [
   n1 = FindGivenAttachName!() -> (then, Find Attach Name:String, New Attach Name:String) {is=title K2Node_FunctionEntry}
   > Set AttachNameToFind(AttachNameToFind:String=@n1) -> (Output_Get:String)
   > n2 = GetSelectedAssets()
   > Branch(Condition=Is Not Empty(TargetArray=@n2))
   > n3 = Set SelectedAssets(SelectedAssets:Array of Object References=@n2) -> (Output_Get:Array of Object References)
   > n4 = For Each Loop(Array=@n3)
   n5 = CreateDialogueAndFindPaths!(
    execute=@n4.Completed
    self
    DB Text:String="The following sequences have been found with attach name: "
    Array:Array of Object References=Get AnimSequences() -> (AnimSequences:Array of Object References)
    MessageType:EAppMsgType Enum=YesNo
    DB Prompt:String="\r\n\r\nWould you like to replace the attach names with the following value?\r\n"
    Replace Attach Name:String=Get New Attach Name() -> (New Attach Name:String)
   ) -> (then, FoundObjectPaths:Array of Strings, PromptResponse:EAppReturnType Enum) {is=title K2Node_CallFunction}
   > SyncBrowserToObjects(AssetPaths=@n5)
   > Clear(TargetArray:Array of Object References=Get AnimSequences())
   > Branch(Condition="Equal (Enum)"(A:EAppReturnType Enum=@n5.PromptResponse, B:EAppReturnType Enum=Yes))
   > ReplaceNotifyAttachNames!(execute, self, New Attach Name:String=Get New Attach Name()) -> (then)
   n6 = Cast To AnimSequenceBase!(execute=@n4.LoopBody, Object=@n4) -> (then, CastFailed, AsAnim Sequence Base:Anim Sequence Base Object Reference) {is=title K2Node_DynamicCast}
   > Set CurrentAnimAsset(CurrentAnimAsset:Object Reference=@n6) -> (Output_Get:Object Reference)
   > n7 = Set NotifyEvents(NotifyEvents:Array of Anim Notify Event Structures=GetAnimationNotifyEvents!(AnimationSequenceBase:Anim Sequence Base Object Reference=@n6) -> (NotifyEvents:Array of Anim Notify Event Structures)) -> (Output_Get:Array of Anim Notify Event Structures)
   n8 = For Each Loop(
    Exec=[@n7.then, @n11.then]
    Array:Array of Anim Notify Event Structures=Get NotifyEvents() -> (NotifyEvents:Array of Anim Notify Event Structures)
   ) -> (Array Element:Anim Notify Event Structure)
   Branch(execute=@n8.Completed, Condition=Is Not Empty(TargetArray:Array of Anim Notify Event Structures=Get NotifiesToUpdate()))
   > Add(
    TargetArray:Array of Object References=Get AnimSequences()
    NewItem:"Object Reference (by ref)"=Get CurrentAnimAsset() -> (CurrentAnimAsset:Object Reference)
   )
   > Clear(TargetArray:Array of Anim Notify Event Structures=Get NotifiesToUpdate())
   n9 = GetAnimNotifyContext!(execute=@n8.LoopBody, self, CurrentNotify:Anim Notify Event Structure=@n8) -> (then, AsAnimation Notify Context:Animation Notify Context Object Reference)
   > FindAnimNotifiesWithJawOrPelvisUsage!(execute, self, NotifyContext:Animation Notify Context Object Reference=@n9) -> (then) {is=title K2Node_CallFunction}
   n10 = Cast To AnimMontage!(execute=@n6.CastFailed, Object=@n4) -> (then, CastFailed, AsAnim Montage:Anim Montage Object Reference) {is=title K2Node_DynamicCast}
   > Set CurrentAnimAsset(CurrentAnimAsset=@n10)
   > n11 = Set NotifyEvents(NotifyEvents=GetAnimationNotifyEvents(AnimationSequenceBase=@n10))
  ]}
 GetAnimNotifyContext {nodes [
   n1 = GetAnimNotifyContext!() -> (then, CurrentNotify:Anim Notify Event Structure) {is=title K2Node_FunctionEntry}
   > Set Current Notify(Current Notify:Anim Notify Event Structure=@n1) -> (Output_Get:Anim Notify Event Structure)
   > n2 = Cast To AnimNotify_SoundLibrary!(
    execute
    Object=Break Anim Notify Event!(AnimNotifyEvent:"Anim Notify Event Structure (by ref)"=@n1) -> (
     TriggerWeightThreshold:float=~
     NotifyName:Name
     Notify:Anim Notify Object Reference
     NotifyStateClass:Anim Notify State Object Reference
     MontageTickType:EMontageNotifyTickType Enum
     NotifyTriggerChance:float=~
     NotifyFilterType:ENotifyFilterType Enum
     NotifyFilterLOD:int
     bCanBeFilteredViaRequest:bool=~
     bTriggerOnDedicatedServer:bool=~
     bTriggerOnFollower:bool=~
    ) {is=title K2Node_BreakStruct}.Notify
   ) -> (then, CastFailed, AsAnim Notify Sound Library:Anim Notify Sound Library Object Reference) {is=title K2Node_DynamicCast}
   > n3 = GetEditorProperty!(execute:Exec, Object:Object Reference=@n2, PropertyName:Name=SoundContext) -> (then:Exec, PropertyValue:Object Reference, ReturnValue:bool)
   > Branch(Condition=@n3.ReturnValue)
   > n4 = Cast To SoundLibraryAnimContext!(execute, Object=@n3) -> (then, CastFailed, AsAnimation Notify Context:Animation Notify Context Object Reference) {is=title K2Node_DynamicCast}
   > Return Node!(execute, AsAnimation Notify Context:Animation Notify Context Object Reference=@n4)
  ]}
 ReplaceNotifyAttachNames {nodes [
   n1 = Cast To AnimSequence!(execute, Object:Wildcard) -> (then, CastFailed, AsAnim Sequence:Anim Sequence Object Reference)
   > n2 = DoesBoneNameExist!(execute, AnimationSequence:Anim Sequence Object Reference=@n1, BoneName:Name=String To Name(InString=Get New Attach Name())) -> (then, bExists:bool)
   > n3 = Branch(Condition=@n2)
   n4 = Get New Attach Name()
   Show Message Dialog(
    execute=@n3.else
    Title=MakeLiteralText(Value="LOCTEXT(\"New Attach Bone Not Found\")")
    Message="To Text (String)"(InString=Append(A=@n4, B=" Is Not a Valid Bone"))
    MessageType=Ok
    DefaultValue=No
    MessageCategory=Warning
   )
   ReplaceNotifyAttachNames!() -> (then, New Attach Name:String) {is=title K2Node_FunctionEntry}
   > n5 = GetSelectedAssets()
   > Branch(Condition=Is Not Empty(TargetArray:Array of Object References=@n5))
   > n6 = Set SelectedAssets(SelectedAssets=@n5)
   > n7 = For Each Loop(Array:Array of Object References=@n6) -> (Array Element:Object Reference)
   CreateDialogueAndFindPaths(
    execute=@n7.Completed
    DB Text=The following sequences have been edited.
    Array=Get AnimSequences()
    MessageType=Ok
    DB Prompt=Their corresponding notifies are updated to have the following attach name:
    Replace Attach Name=Get New Attach Name()
   )
   n8 = Get AnimSequences()
   > SaveLoadedAssets(AssetsToSave=@n8, bOnlyIfIsDirty=true)
   > Clear(TargetArray:Array of Object References=@n8)
   n9 = Cast To AnimSequenceBase(execute=@n7.LoopBody, Object=@n7)
   > Set CurrentAnimAsset(CurrentAnimAsset=@n9)
   > n10 = Set NotifyEvents(NotifyEvents=GetAnimationNotifyEvents(AnimationSequenceBase=@n9))
   n11 = For Each Loop(Exec=[@n19.then, @n10.then], Array:Array of Anim Notify Event Structures=Get NotifyEvents()) -> (Array Element:Anim Notify Event Structure)
   n12 = Get NotifiesToUpdate()
   Branch(execute=@n11.Completed, Condition=Is Not Empty(TargetArray:Array of Anim Notify Event Structures=@n12))
   > TransactObject(Object=Get CurrentAnimAsset())
   > n13 = For Each Loop(Array=@n12)
   Clear(execute=@n13.Completed, TargetArray:Array of Anim Notify Event Structures=Get NotifiesToUpdate())
   > Add(TargetArray=Get AnimSequences(), NewItem=Get CurrentAnimAsset())
   n14 = GetAnimNotifyContext(execute=@n13.LoopBody, CurrentNotify=@n13) {is=title K2Node_CallFunction}
   n15 = Break Animation Notify Settings(SoundLibraryAnimContextSettings=Get AnimSettings(self=@n14)) -> (bOnThisActor=~, bAttachToActor=~, bStopWhenAttachedDestroyed=~, Offset=~, FadeTime=~, NotifyTriggerChance=~, VolumeMultiplier=~)
   n16 = Get New Attach Name()
   > Set AnimSettings!(
    execute
    AnimSettings:Animation Notify Settings Structure=Make Animation Notify Settings!(
     AnimRateThreshold:Float Interval Structure=@n15
     EventName:Gameplay Tag Structure=@n15.EventName
     bOnThisActor:bool=@n15.bOnThisActor
     bAttachToActor:bool=@n15.bAttachToActor
     bStopWhenAttachedDestroyed:bool=@n15.bStopWhenAttachedDestroyed
     Offset:Vector=@n15.Offset
     AttachName:Name=String To Name(InString=@n16)
     FadeTime:float=@n15.FadeTime
     NotifyTriggerChance:float=@n15.NotifyTriggerChance
     VolumeMultiplier:float=@n15.VolumeMultiplier
    ) -> (SoundLibraryAnimContextSettings:Animation Notify Settings Structure) {is=title K2Node_MakeStruct}
    self:Animation Notify Context Object Reference=@n14
   ) -> (then, Output_Get:Animation Notify Settings Structure)
   n17 = GetAnimNotifyContext(execute=@n11.LoopBody, CurrentNotify=@n11)
   > FindAnimNotifiesWithJawOrPelvisUsage(NotifyContext=@n17)
   n18 = Cast To AnimMontage(execute=@n9.CastFailed, Object=@n7)
   > Set CurrentAnimAsset(CurrentAnimAsset=@n18)
   > n19 = Set NotifyEvents(NotifyEvents=GetAnimationNotifyEvents(AnimationSequenceBase=@n18))
  ]}
]
variables [
 AnimSequences {type="array<Object>", default []}
 AttachName {type=string, default=None}
 AttachNamesArray {type="array<string>", default []}
 AttachNameToFind {type=string, default=""}
 Current Notify {
  type=AnimNotifyEvent
  default {
   bCanBeFilteredViaRequest=true
   bConvertedFromBranchingPoint=false
   bTriggerOnDedicatedServer=true
   bTriggerOnFollower=false
   cachedLinkMethod=Absolute
   duration=0
   endTriggerTimeOffset=0
   guid=#1
   linkMethod=Absolute
   linkValue=0
   linkedMontage=null
   linkedSequence=null
   montageTickType=Queued
   notify=null
   notifyFilterLOD=0
   notifyFilterType=NoFiltering
   notifyName=None
   notifyStateClass=null
   notifyTriggerChance=1
   segmentBeginTime=0
   segmentIndex=-1
   segmentLength=0
   slotIndex=0
   trackIndex=0
   triggerTimeOffset=0
   triggerWeightThreshold=0
   endLink {
    cachedLinkMethod=Absolute
    linkMethod=Absolute
    linkValue=0
    linkedMontage=null
    linkedSequence=null
    segmentBeginTime=0
    segmentIndex=-1
    segmentLength=0
    slotIndex=0
   }
   notifyColor=<rgba 0 0 0 1>
  }
 }
 CurrentAnimAsset {type=Object, default=null}
 NotifiesToUpdate {type="array<AnimNotifyEvent>", default []}
 NotifyEvents {type="array<AnimNotifyEvent>", default []}
 SelectedAssets {type="array<Object>", default []}
 Skeleton {type=Skeleton, default=null}
]
